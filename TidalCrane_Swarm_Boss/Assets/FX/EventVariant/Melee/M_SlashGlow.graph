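# Unlit, translucent, two-sided glow material for Niagara mesh particles, named for melee slash effects.
# Particle color feeds an emissive adjustment function, while its alpha is multiplied by U and V edge masks and a squared view-angle fade to form opacity.
# A tangent-basis Y offset, scaled by centered V and the dynamic Width parameter, widens the mesh per particle.
# Scalar parameter T is unconnected.
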Material /TidalCrane_Swarm_Boss/Assets/FX/EventVariant/Melee/M_SlashGlow
graph {
 properties {
  bAutomaticallySetUsageInEditor=false
  bUsedWithNiagaraMeshParticles=true
  blendMode=BLEND_Translucent
  shadingModel=MSM_Unlit
  twoSided=true
  usedShadingModels=Unlit
 }
 nodes [
  n1 = ParticleColor()
  MaterialFunctionCall(Emissive=@n1.RGB) {materialFunction=&MF_FX_EmissiveAdjustments}
  n2 = MaterialFunctionCall(Float2=TextureCoordinate()) {materialFunction=&BreakOutFloat2Components}
  n3 = Multiply(A=@n2.G) {constB=2}
  Multiply(
   A=MaterialFunctionCall() {materialFunction=&TangentBasis}.Y V
   B=Multiply(A=Subtract(A=@n3), B=DynamicParameter() {defaultValue=<rgba 60 1 1 1>, paramNames [Width,Param2,Param3,Param4]}.Width)
  )
  n4 = VectorParameter() {group=EdgeF, parameterName=EdgeF_EdgeMask_Control, channelNames {a="", b=Power, g=Multiply, r=Offset}, defaultValue=<rgba 0 1 1 1>}
  n5 = MaterialFunctionCall!(UorV=@n2.G, EdgeControl=@n4) {materialFunction=/Game/Effects/Niagara/MasterMaterials/Functions/Common/FXCommon_EdgeMask_Simple}
  n6 = VectorParameter() {group=EdgeZ, parameterName=EdgeZ_EdgeMask_InvertControl, channelNames {a="", b=Power, g=Multiply, r=Offset}, defaultValue=<rgba 0 1 1 1>}
  n7 = VectorParameter() {group=EdgeZ, parameterName=EdgeZ_EdgeMask_Control, channelNames {a="", b=Power, g=Multiply, r=Offset}, defaultValue=<rgba 0 1 1 1>}
  n8 = MaterialFunctionCall!(UorV=@n2.R, Invert_EdgeControl=@n6, NoInvert_EdgeControl=@n7) {materialFunction=/Game/Effects/Niagara/MasterMaterials/Functions/Common/FXCommon_EdgeMask_Ex}
  n9 = ScalarParameter() {defaultValue=0.5, parameterName=FSVA}
  Multiply(
   A=Multiply(A=@n1.A, B=Multiply(A=@n5.Mask, B=@n8.Mask))
   B=Power(Base=MaterialFunctionCall("FadeSensitivity (optional)"=@n9) {materialFunction=&FadeBasedOnViewAngle}.Result)
  )
  ScalarParameter() {defaultValue=300, parameterName=T}
 ]
}
