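# Niagara system for the Masamune big slash release: a brief purple-magenta spinning slash with sparks and airflow.
# HighLight001 merges the disabled Reamin, HighLight, Glow and Glow001 emitters into four spawn groups, switching color, scale and curves per group and writing VisibilityTag for its four mesh renderers.
# FollowParticlesFixed sweeps GPU sprites along a circular spline mesh, launching velocity-stretched streaks with drag.
# SoftAirflow adds expanding, fading meshes tinted by time of day.
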NiagaraSystem /TidalCrane_Swarm_Boss/Assets/FX/EventVariant/Melee/NS_Masamune_BigSlashRelease
dependencies {
 usedDataInterfaces [&NiagaraDataInterfaceCurve, &NiagaraDataInterfaceColorCurve, &NiagaraDataInterfaceStaticMesh]
 usedDynamicInputs []
 usedModules [
  &self:SystemSpawnScript.SystemScriptSource.SystemScriptGraph.NiagaraNodeAssignment_0.SetVariables_#1
  &self:SystemSpawnScript.SystemScriptSource.SystemScriptGraph.NiagaraNodeAssignment_1.SetVariables_#2
  &self:SystemSpawnScript.SystemScriptSource.SystemScriptGraph.NiagaraNodeAssignment_2.SetVariables_#3
  &self:SystemSpawnScript.SystemScriptSource.SystemScriptGraph.NiagaraNodeAssignment_3.SetVariables_#4
  &SystemState
 ]
}
emitters {
 FollowParticlesFixed {
  simTarget=GPUComputeSim
  emitterUpdateScript [
   /SetVariables(Emitter.LoopDuration=0.15, Emitter.SpawnDuration=0.1) {isSetParameters=true}
   &EmitterState(Life Cycle Mode=::NewEnumerator1, Loop Behavior=::NewEnumerator1, Loop Duration=$Emitter.LoopDuration, UseLoopDelay=false)
   &SpawnRate(SpawnRate=1000, Use Spawn Probability=false)
  ]
  particleSpawnScript [
   &InitializeParticle(
    Color=<rgba 5 0.5 4.6793 1>
    Lifetime=&RandomRangeFloat(Minimum=0.15, Maximum=0.25)
    Sprite Size=&RandomRangeVector2D(Minimum=<0.8 2>, Maximum=<1.2 3.5>)
    Sprite Size Mode=::NewEnumerator4
    UsePositionOffset=false
    Write Material Random=true
   )
   /Game/Effects/Niagara/Modules/Position/FN_PositionAlongSplineMesh(
    Progress:NiagaraFloat=&FloatFromCurve(
     FloatCurve=DataInterface() {propertyValues {Curve {keys [<at 0 0.6>,<at 1 0>]}}}
     CurveIndex=&Multiply_Float(A=$Emitter.NormalizedLoopAge, B=&Divide_Float(A=$Emitter.LoopDuration, B=$Emitter.SpawnDuration))
    )
    QuadCountInOneSegment=2
    Static Mesh=DataInterface() {propertyValues {
      DefaultMesh=/TidalCrane_Swarm_Boss/Assets/FX/EventVariant/Relocation/SM_EmperorGlider_Circle
      InstanceIndex=-1
      LODIndex=0
      PreviewMesh=&SM_EmperorGlider_Circle
      SoftSourceActor=null
      SourceMode=DefaultMeshOnly
      bAllowSamplingFromStreamingLODs=false
      bCaptureTransformsPerFrame=true
      bUsePhysicsBodyVelocity=false
      FilteredSockets []
      LODIndexUserParameter {parameter {name=None, typeDefHandle {registeredTypeIndex=108}, varData []}}
      MeshParameterBinding {parameter {name=None, typeDefHandle {registeredTypeIndex=1}, varData []}}
      SectionFilter {allowedMaterialSlots []}
     }}
   )
   &AddVelocity(
    Cone Angle=10
    Cone Axis=&Multiply_VectorByFloat(Vector=$Particles.FN_PositionAlongSplineMesh.Bitangent, Float=-1)
    Module.UseConeFalloff=false
    Module.UseVelDistribution=false
    Override Seed=false
    Rotation Coordinate Space=::Simulation
    Velocity Mode=::NewEnumerator2
    Velocity Speed=&RandomRangeFloat(Minimum=80, Maximum=534.6359)
   )
   &self:Transform(
    Eular:Vector3f=<0 -90 49.201>
    Scale:Vector3f=&Multiply_Vector(A=<8 8 8>, B=<7.3 7.3 7.3>)
    ScaleV:Vector3f=<5 5 5>
   ) {module=Transform}
  ]
  particleUpdateScript [
   &ParticleState()
   &ApplyOwnerScaleToAttributes(Scale Initial Sprite Size=true, Scale Initial Velocity=true, Sprite Size Scale Amount=3.2519)
   &Drag(Drag=&RandomRangeFloat(Minimum=2.5, Maximum=4.5), Use Linear Drag=true, Use Rotational Drag=true)
   &SolveForcesAndVelocity(Clamp Velocity=false, Limit Acceleration=false)
   &ScaleSpriteSize(
    Scale Factor=&MakeVector2D(
     X=1
     Y=&FloatFromCurve(FloatCurve=DataInterface() {propertyValues {Curve {keys [<at 0.8 1>,<at 1 0>]}}})
    )
    Scale Sprite Size Mode=::NewEnumerator2
   )
   &ScaleSpriteSizeBySpeed(Max Scale Factor=<1 8>, Min Scale Factor=<1 8>, Velocity Threshold=698.8588)
   &FN_Sprite_Attributes(
    Color=&MultiplyLinearColor(Linear Color=$Particles.Color, Scale Factor=<1 1 1 0.01>)
    Create Color Attribute=true
    Create SpriteSize Attribute=true
    SpriteSize=&Multiply_Vector2D(A=$Particles.SpriteSize, B=<30 3>)
    Write to Module Namespace=true
   )
  ]
  rendererClasses [&NiagaraSpriteRendererProperties]
  renderers [
   {rendererClass=&NiagaraSpriteRendererProperties}
   {rendererClass=&NiagaraSpriteRendererProperties, rendererIndex=1}
  ]
 }
 Glow {
  bEnabled=false
  emitterUpdateScript [
   &EmitterState(Life Cycle Mode=::NewEnumerator1, Loop Behavior=::NewEnumerator1, Loop Duration=1, UseLoopDelay=false)
   &SpawnBurst_Instantaneous(Use Spawn Probability=false, UseLoopCountLimit=false)
  ]
  particleSpawnScript [
   &InitializeParticle(
    Color=<rgba 0.1333 0 0.2 10>
    Lifetime=0.5
    Mesh Scale=&Multiply_VectorByFloat(Vector=<1 1 0.25>, Float=2)
    Mesh Scale Mode=::NewEnumerator4
    Position Offset=$System.FX position
    UsePositionOffset=true
    Write Material Random=true
   )
   &InitialMeshOrientation(Mesh Orientation Mode=::NewEnumerator2, Rotation=$System.CustomRotation, Use Rotation Vector=true)
   &InitialMeshOrientation(Mesh Orientation Mode=::NewEnumerator2, Rotation=<0.5 0 0>, Use Rotation Vector=true) {module=InitialMeshOrientation001}
   /SetVariables(Particles.Scale=&Multiply_Vector(A=&Multiply_Vector(A=$Engine.Owner.Scale, B=$System.SizeScale), B=$Particles.Scale)) {isSetParameters=true, module=SetVariables_#5}
  ]
  particleUpdateScript [
   &ParticleState()
   &UpdateMeshOrientation(
    Rotation Rate=&FloatFromCurve(
     FloatCurve=DataInterface() {propertyValues {Curve {keys [{arriveTangent=-3.0754, interpMode=RCIM_Cubic, leaveTangent=-3.0754, tangentMode=RCTM_User, time=0, value=2}, <at 0.5 0.1>]}}}
     Scale Curve=1.4
    )
    Rotation Vector=<0 0 -1>
   ) {module=UpdateMeshOrientation001}
   &ScaleColor(
    Scale Alpha=&FloatFromCurve(FloatCurve=DataInterface() {propertyValues {Curve {keys [
         <at 0 1 RCIM_Cubic>
         {arriveTangent=-4.2795, interpMode=RCIM_Cubic, leaveTangent=-4.2795, tangentMode=RCTM_User, time=0.5844, value=0.5146}
         {arriveTangent=-2.592, interpMode=RCIM_Cubic, leaveTangent=-2.592, tangentMode=RCTM_User, time=1, value=0}
        ]}}})
    Scale RGB=&Multiply_VectorByFloat(
     Vector=&MakeVectorFromLinearColorRGB(Color=&ColorFromCurve(ColorCurve=DataInterface() {propertyValues {
         AlphaCurve {keys []}
         BlueCurve {keys [<at 0 1 RCIM_Cubic>,<at 0.8853 2 RCIM_Cubic>]}
         GreenCurve {keys [<at 0 0.05 RCIM_Cubic>,<at 0.8853 0.4 RCIM_Cubic>]}
         RedCurve {keys [<at 0 1 RCIM_Cubic>,<at 0.8853 1.7333 RCIM_Cubic>]}
        }}))
     Float=&FloatFromCurve(FloatCurve=DataInterface() {propertyValues {Curve {keys [<at 0 1>,<at 0.5 1>]}}}, Scale Curve=3)
    )
    ScaleA=true
    ScaleRGB=true
   )
   &DynamicMaterialParameters(
    Index 0 Param 1=&FloatFromCurve(FloatCurve=DataInterface() {propertyValues {Curve {keys [<at 0 0 RCIM_Cubic>,<at 1 1>]}}})
    Index 0 Param 2=0.1
    Module.Index0Param0Write=true
    Module.Index0Param1Write=true
    Module.Index0Param2Write=true
    Module.Index0Param3Write=true
   )
  ]
  rendererClasses [&NiagaraMeshRendererProperties]
  renderers [{rendererClass=&NiagaraMeshRendererProperties, rendererIndex=0}]
 }
 Glow001 {
  bEnabled=false
  emitterUpdateScript [
   &EmitterState(Life Cycle Mode=::NewEnumerator1, Loop Behavior=::NewEnumerator1, Loop Duration=1, UseLoopDelay=false)
   &SpawnBurst_Instantaneous(Use Spawn Probability=false, UseLoopCountLimit=false)
  ]
  particleSpawnScript [
   &InitializeParticle(
    Color=<rgba 0.3333 0 0.5 1>
    Lifetime=0.5
    Mesh Scale=&Multiply_VectorByFloat(Vector=<1 1 0.25>, Float=2)
    Mesh Scale Mode=::NewEnumerator4
    Position Offset=$System.FX position
    UsePositionOffset=true
    Write Material Random=true
   )
   &InitialMeshOrientation(Mesh Orientation Mode=::NewEnumerator2, Rotation=$System.CustomRotation, Use Rotation Vector=true)
   &InitialMeshOrientation(Mesh Orientation Mode=::NewEnumerator2, Rotation=<0.5 0 0>, Use Rotation Vector=true) {module=InitialMeshOrientation001}
   /SetVariables(Particles.Scale=&Multiply_Vector(A=&Multiply_Vector(A=$Engine.Owner.Scale, B=$System.SizeScale), B=$Particles.Scale)) {isSetParameters=true, module=SetVariables_#5}
  ]
  particleUpdateScript [
   &ParticleState()
   &ScaleMeshSize(Scale Factor=&Multiply_Vector(A=$Engine.Owner.Scale, B=$System.SizeScale)) {enabled=false}
   &UpdateMeshOrientation(
    Rotation Rate=&FloatFromCurve(
     FloatCurve=DataInterface() {propertyValues {Curve {keys [{arriveTangent=-3.0754, interpMode=RCIM_Cubic, leaveTangent=-3.0754, tangentMode=RCTM_User, time=0, value=2}, <at 0.5 0.1>]}}}
     Scale Curve=1.4
    )
    Rotation Vector=<0 0 -1>
   ) {module=UpdateMeshOrientation001}
   &ScaleColor(
    Scale Alpha=&FloatFromCurve(FloatCurve=DataInterface() {propertyValues {Curve {keys [
         <at 0 1 RCIM_Cubic>
         {arriveTangent=-4.2795, interpMode=RCIM_Cubic, leaveTangent=-4.2795, tangentMode=RCTM_User, time=0.5844, value=0.5146}
         {arriveTangent=-2.592, interpMode=RCIM_Cubic, leaveTangent=-2.592, tangentMode=RCTM_User, time=1, value=0}
        ]}}})
    Scale RGB=&Multiply_VectorByFloat(
     Vector=&MakeVectorFromLinearColorRGB(Color=&ColorFromCurve(ColorCurve=DataInterface() {propertyValues {
         AlphaCurve {keys []}
         BlueCurve {keys [<at 0 2 RCIM_Cubic>,<at 0.312 1>,<at 0.8853 2 RCIM_Cubic>]}
         GreenCurve {keys [<at 0 0.1 RCIM_Cubic>,<at 0.312 0.0533>,<at 0.8853 0 RCIM_Cubic>]}
         RedCurve {keys [<at 0 2 RCIM_Cubic>,<at 0.312 0.997>,<at 0.8853 1.6667 RCIM_Cubic>]}
        }}))
     Float=&FloatFromCurve(FloatCurve=DataInterface() {propertyValues {Curve {keys [<at 0 1>,<at 0.5 1>]}}}, Scale Curve=3)
    )
    ScaleA=true
    ScaleRGB=true
   )
   &DynamicMaterialParameters(
    Index 0 Param 1=&Multiply_Float(A=60, B=&MakeFloatFromVector(Vector=$Engine.Owner.Scale, Channel=::NewEnumerator0))
    Index 0 Param 2=0.1
    Module.Index0Float0Write=true
    Module.Index0Float1Write=true
    Module.Index0Float2Write=true
    Module.Index0Float3Write=true
   )
  ]
  rendererClasses [&NiagaraMeshRendererProperties]
  renderers [{rendererClass=&NiagaraMeshRendererProperties, rendererIndex=0}]
 }
 HighLight {
  bEnabled=false
  emitterUpdateScript [
   &EmitterState(Life Cycle Mode=::NewEnumerator1, Loop Behavior=::NewEnumerator1, Loop Duration=1, UseLoopDelay=false)
   &SpawnBurst_Instantaneous(Use Spawn Probability=false, UseLoopCountLimit=false)
  ]
  particleSpawnScript [
   &InitializeParticle(
    Lifetime=0.3
    Mesh Scale=&Multiply_VectorByFloat(Vector=$System.WeaponSize, Float=2.2)
    Mesh Scale Mode=::NewEnumerator4
    Position Offset=$System.FX position
    UsePositionOffset=true
    Write Material Random=true
   )
   &InitialMeshOrientation(Mesh Orientation Mode=::NewEnumerator2, Rotation=$System.CustomRotation, Use Rotation Vector=true)
   &InitialMeshOrientation(Mesh Orientation Mode=::NewEnumerator2, Rotation=<0.5 0 0>, Use Rotation Vector=true) {module=InitialMeshOrientation001}
   /SetVariables(Particles.Scale=&Multiply_Vector(A=&Multiply_Vector(A=$Engine.Owner.Scale, B=$System.SizeScale), B=$Particles.Scale)) {isSetParameters=true, module=SetVariables_#6}
  ]
  particleUpdateScript [
   &ParticleState()
   &ScaleMeshSize(Scale Factor=&Multiply_Vector(A=$Engine.Owner.Scale, B=$System.SizeScale)) {enabled=false}
   &UpdateMeshOrientation(
    Rotation Rate=&FloatFromCurve(FloatCurve=DataInterface() {propertyValues {Curve {keys [{arriveTangent=-3.0754, interpMode=RCIM_Cubic, leaveTangent=-3.0754, tangentMode=RCTM_User, time=0, value=2}, <at 1 0.1>]}}})
    Rotation Vector=<0 0 -1>
   ) {module=UpdateMeshOrientation001}
   &ScaleColor(
    Scale Alpha=&FloatFromCurve(FloatCurve=DataInterface() {propertyValues {Curve {keys [{interpMode=RCIM_Cubic, tangentMode=RCTM_User, time=0, value=1}, {interpMode=RCIM_Cubic, tangentMode=RCTM_User, time=1, value=0}]}}})
    Scale RGB=&Multiply_VectorByFloat(
     Vector=&MakeVectorFromLinearColorRGB(Color=&ColorFromCurve(ColorCurve=DataInterface() {propertyValues {
         AlphaCurve {keys []}
         BlueCurve {keys [<at 0.3184 99.9999 RCIM_Cubic>,<at 0.4968 0.7135 RCIM_Cubic>]}
         GreenCurve {keys [<at 0.3184 0 RCIM_Cubic>,<at 0.4968 0 RCIM_Cubic>]}
         RedCurve {keys [<at 0.3184 100 RCIM_Cubic>,<at 0.4968 0.4757 RCIM_Cubic>]}
        }}))
     Float=2
    )
    ScaleA=true
    ScaleRGB=true
   )
   &DynamicMaterialParameters(
    Index 0 Param 1=&FloatFromCurve(FloatCurve=DataInterface() {propertyValues {Curve {keys [{tangentMode=RCTM_User, time=0, value=0}, {tangentMode=RCTM_User, time=1, value=1}]}}})
    Index 0 Param 2=0.6
    Module.Index0Param0Write=true
    Module.Index0Param1Write=true
    Module.Index0Param2Write=true
    Module.Index0Param3Write=true
   )
  ]
  rendererClasses [&NiagaraMeshRendererProperties]
  renderers [{rendererClass=&NiagaraMeshRendererProperties, rendererIndex=0}]
 }
 HighLight001 {
  emitterUpdateScript [
   &EmitterState(Life Cycle Mode=::NewEnumerator1, Loop Behavior=::NewEnumerator1, Loop Duration=1, UseLoopDelay=false)
   &SpawnBurst_Instantaneous(Use Spawn Probability=false, UseLoopCountLimit=false)
   &SpawnBurst_Instantaneous(Spawn Group=1, Use Spawn Probability=false, UseLoopCountLimit=false) {module=SpawnBurst_Instantaneous001}
   &SpawnBurst_Instantaneous(Spawn Group=2, Use Spawn Probability=false, UseLoopCountLimit=false) {module=SpawnBurst_Instantaneous002}
   &SpawnBurst_Instantaneous(Spawn Group=3, Use Spawn Probability=false, UseLoopCountLimit=false) {module=SpawnBurst_Instantaneous003}
  ]
  particleSpawnScript [
   &InitializeParticle(
    Color=&MergeHelper_ColorSwitch(
     0=<rgba 1 1 1 1>
     1=<rgba 1 1 1 1>
     2=<rgba 0.1333 0 0.2 10>
     3=<rgba 0.3333 0 0.5 1>
     4=<rgba 0 0 0 0>
     Index=&SpawnGroup()
     Num=::NewEnumerator3
    )
    Lifetime=&MergeHelper_FloatSwitch(0=0.3, 1=0.6, 2=0.5, 3=0.5, 4=0, Index=&SpawnGroup(), Num=::NewEnumerator3)
    Mesh Scale=&MergeHelper_VectorSwitch(
     0=&Multiply_VectorByFloat(Vector=$System.WeaponSize, Float=2.2)
     1=&Multiply_VectorByFloat(Vector=<1 1 0.25>, Float=2)
     2=&Multiply_VectorByFloat(Vector=<1 1 0.25>, Float=2)
     3=&Multiply_VectorByFloat(Vector=<1 1 0.25>, Float=2)
     4=<0 0 0>
     Index=&SpawnGroup()
     Num=::NewEnumerator3
    )
    Mesh Scale Mode=::NewEnumerator4
    Position Offset=$System.FX position
    UsePositionOffset=true
    Write Material Random=true
   )
   &InitialMeshOrientation(Mesh Orientation Mode=::NewEnumerator2, Rotation=$System.CustomRotation, Use Rotation Vector=true)
   &InitialMeshOrientation(Mesh Orientation Mode=::NewEnumerator2, Rotation=<0.5 0 0>, Use Rotation Vector=true) {module=InitialMeshOrientation001}
   /SetVariables(Particles.Scale=&Multiply_Vector(A=&Multiply_Vector(A=$Engine.Owner.Scale, B=$System.SizeScale), B=$Particles.Scale)) {isSetParameters=true, module=SetVariables_#6}
   /SetVariables(Particles.VisibilityTag=&SpawnGroup()) {isSetParameters=true}
  ]
  particleUpdateScript [
   &ParticleState()
   &UpdateMeshOrientation(
    Rotation Rate=&MergeHelper_FloatSwitch(
     0=&FloatFromCurve(FloatCurve=DataInterface() {propertyValues {Curve {keys [{arriveTangent=-3.0754, interpMode=RCIM_Cubic, leaveTangent=-3.0754, tangentMode=RCTM_User, time=0, value=2}, <at 1 0.1>]}}})
     1=&FloatFromCurve(
      FloatCurve=DataInterface() {propertyValues {Curve {keys [{arriveTangent=-3.0754, interpMode=RCIM_Cubic, leaveTangent=-3.0754, tangentMode=RCTM_User, time=0, value=2}, <at 0.5 0.1>]}}}
      Scale Curve=1.4
     )
     2=&FloatFromCurve(
      FloatCurve=DataInterface() {propertyValues {Curve {keys [{arriveTangent=-3.0754, interpMode=RCIM_Cubic, leaveTangent=-3.0754, tangentMode=RCTM_User, time=0, value=2}, <at 0.5 0.1>]}}}
      Scale Curve=1.4
     )
     3=&FloatFromCurve(
      FloatCurve=DataInterface() {propertyValues {Curve {keys [{arriveTangent=-3.0754, interpMode=RCIM_Cubic, leaveTangent=-3.0754, tangentMode=RCTM_User, time=0, value=2}, <at 0.5 0.1>]}}}
      Scale Curve=1.4
     )
     4=0
     Index=&SpawnGroup()
     Num=::NewEnumerator3
    )
    Rotation Vector=<0 0 -1>
   ) {module=UpdateMeshOrientation001}
   &ScaleColor(
    Scale Alpha=&MergeHelper_FloatSwitch(
     0=&FloatFromCurve(FloatCurve=DataInterface() {propertyValues {Curve {keys [{interpMode=RCIM_Cubic, tangentMode=RCTM_User, time=0, value=1}, {interpMode=RCIM_Cubic, tangentMode=RCTM_User, time=1, value=0}]}}})
     1=&FloatFromCurve(
      FloatCurve=DataInterface() {propertyValues {Curve {keys [
          <at 0 1 RCIM_Cubic>
          {arriveTangent=-4.2795, interpMode=RCIM_Cubic, leaveTangent=-4.2795, tangentMode=RCTM_User, time=0.5844, value=0.5146}
          {arriveTangent=-2.592, interpMode=RCIM_Cubic, leaveTangent=-2.592, tangentMode=RCTM_User, time=1, value=0}
         ]}}}
      Scale Curve=3
     )
     2=&FloatFromCurve(FloatCurve=DataInterface() {propertyValues {Curve {keys [
          <at 0 1 RCIM_Cubic>
          {arriveTangent=-4.2795, interpMode=RCIM_Cubic, leaveTangent=-4.2795, tangentMode=RCTM_User, time=0.5844, value=0.5146}
          {arriveTangent=-2.592, interpMode=RCIM_Cubic, leaveTangent=-2.592, tangentMode=RCTM_User, time=1, value=0}
         ]}}})
     3=&FloatFromCurve(FloatCurve=DataInterface() {propertyValues {Curve {keys [
          <at 0 1 RCIM_Cubic>
          {arriveTangent=-4.2795, interpMode=RCIM_Cubic, leaveTangent=-4.2795, tangentMode=RCTM_User, time=0.5844, value=0.5146}
          {arriveTangent=-2.592, interpMode=RCIM_Cubic, leaveTangent=-2.592, tangentMode=RCTM_User, time=1, value=0}
         ]}}})
     4=0
     Index=&SpawnGroup()
     Num=::NewEnumerator3
    )
    Scale RGB=&MergeHelper_VectorSwitch(
     0=&Multiply_VectorByFloat(
      Vector=&MakeVectorFromLinearColorRGB(Color=&ColorFromCurve(ColorCurve=DataInterface() {propertyValues {
          AlphaCurve {keys []}
          BlueCurve {keys [<at 0.3184 99.9999 RCIM_Cubic>,<at 0.4968 0.7135 RCIM_Cubic>]}
          GreenCurve {keys [<at 0.3184 0 RCIM_Cubic>,<at 0.4968 0 RCIM_Cubic>]}
          RedCurve {keys [<at 0.3184 100 RCIM_Cubic>,<at 0.4968 0.4757 RCIM_Cubic>]}
         }}))
      Float=2
     )
     1=&Multiply_VectorByFloat(
      Vector=&MakeVectorFromLinearColorRGB(Color=&ColorFromCurve(ColorCurve=DataInterface() {propertyValues {
          AlphaCurve {keys []}
          BlueCurve {keys [<at 0 2 RCIM_Cubic>,<at 0.8853 2 RCIM_Cubic>]}
          GreenCurve {keys [<at 0 0.3 RCIM_Cubic>,<at 0.8853 0.4 RCIM_Cubic>]}
          RedCurve {keys [<at 0 1.7167 RCIM_Cubic>,<at 0.8853 1.4667 RCIM_Cubic>]}
         }}))
      Float=&FloatFromCurve(FloatCurve=DataInterface() {propertyValues {Curve {keys [<at 0 1>,<at 0.5 1>]}}}, Scale Curve=3)
     )
     2=&Multiply_VectorByFloat(
      Vector=&MakeVectorFromLinearColorRGB(Color=&ColorFromCurve(ColorCurve=DataInterface() {propertyValues {
          AlphaCurve {keys []}
          BlueCurve {keys [<at 0 1 RCIM_Cubic>,<at 0.8853 2 RCIM_Cubic>]}
          GreenCurve {keys [<at 0 0.05 RCIM_Cubic>,<at 0.8853 0.4 RCIM_Cubic>]}
          RedCurve {keys [<at 0 1 RCIM_Cubic>,<at 0.8853 1.7333 RCIM_Cubic>]}
         }}))
      Float=&FloatFromCurve(FloatCurve=DataInterface() {propertyValues {Curve {keys [<at 0 1>,<at 0.5 1>]}}}, Scale Curve=3)
     )
     3=&Multiply_VectorByFloat(
      Vector=&MakeVectorFromLinearColorRGB(Color=&ColorFromCurve(ColorCurve=DataInterface() {propertyValues {
          AlphaCurve {keys []}
          BlueCurve {keys [<at 0 2 RCIM_Cubic>,<at 0.312 1>,<at 0.8853 2 RCIM_Cubic>]}
          GreenCurve {keys [<at 0 0.1 RCIM_Cubic>,<at 0.312 0.0533>,<at 0.8853 0 RCIM_Cubic>]}
          RedCurve {keys [<at 0 2 RCIM_Cubic>,<at 0.312 0.997>,<at 0.8853 1.6667 RCIM_Cubic>]}
         }}))
      Float=&FloatFromCurve(FloatCurve=DataInterface() {propertyValues {Curve {keys [<at 0 1>,<at 0.5 1>]}}}, Scale Curve=3)
     )
     4=<0 0 0>
     Index=&SpawnGroup()
     Num=::NewEnumerator3
    )
    ScaleA=true
    ScaleRGB=true
   )
   &DynamicMaterialParameters(
    Index 0 Param 1=&MergeHelper_FloatSwitch(
     0=&FloatFromCurve(FloatCurve=DataInterface() {propertyValues {Curve {keys [{tangentMode=RCTM_User, time=0, value=0}, {tangentMode=RCTM_User, time=1, value=1}]}}})
     1=&FloatFromCurve(FloatCurve=DataInterface() {propertyValues {Curve {keys [<at 0 0 RCIM_Cubic>,<at 1 1>]}}})
     2=&FloatFromCurve(FloatCurve=DataInterface() {propertyValues {Curve {keys [<at 0 0 RCIM_Cubic>,<at 1 1>]}}})
     3=&Multiply_Float(A=60, B=&MakeFloatFromVector(Vector=$Engine.Owner.Scale, Channel=::NewEnumerator0))
     4=0
     Index=&SpawnGroup()
     Num=::NewEnumerator3
    )
    Index 0 Param 2=&MergeHelper_FloatSwitch(0=0.6, 1=0.1, 2=0.1, 3=0.1, 4=0, Index=&SpawnGroup(), Num=::NewEnumerator3)
    Module.Index0Param0Write=true
    Module.Index0Param1Write=true
    Module.Index0Param2Write=true
    Module.Index0Param3Write=true
   )
  ]
  rendererClasses [&NiagaraMeshRendererProperties]
  renderers [
   {rendererClass=&NiagaraMeshRendererProperties, rendererIndex=0}
   {rendererClass=&NiagaraMeshRendererProperties, rendererIndex=1}
   {rendererClass=&NiagaraMeshRendererProperties, rendererIndex=2}
   {rendererClass=&NiagaraMeshRendererProperties, rendererIndex=3}
  ]
 }
 Reamin {
  bEnabled=false
  emitterUpdateScript [
   &EmitterState(Life Cycle Mode=::NewEnumerator1, Loop Behavior=::NewEnumerator1, Loop Duration=1, UseLoopDelay=false)
   &SpawnBurst_Instantaneous(Use Spawn Probability=false, UseLoopCountLimit=false)
  ]
  particleSpawnScript [
   &InitializeParticle(
    Lifetime=0.6
    Mesh Scale=&Multiply_VectorByFloat(Vector=<1 1 0.25>, Float=2)
    Mesh Scale Mode=::NewEnumerator4
    Position Offset=$System.FX position
    UsePositionOffset=true
    Write Material Random=true
   )
   &InitialMeshOrientation(Mesh Orientation Mode=::NewEnumerator2, Rotation=$System.CustomRotation, Use Rotation Vector=true)
   &InitialMeshOrientation(Mesh Orientation Mode=::NewEnumerator2, Rotation=<0.5 0 0>, Use Rotation Vector=true) {module=InitialMeshOrientation001}
   /SetVariables(Particles.Scale=&Multiply_Vector(A=&Multiply_Vector(A=$Engine.Owner.Scale, B=$System.SizeScale), B=$Particles.Scale)) {isSetParameters=true}
  ]
  particleUpdateScript [
   &ParticleState()
   &UpdateMeshOrientation(
    Rotation Rate=&FloatFromCurve(
     FloatCurve=DataInterface() {propertyValues {Curve {keys [{arriveTangent=-3.0754, interpMode=RCIM_Cubic, leaveTangent=-3.0754, tangentMode=RCTM_User, time=0, value=2}, <at 0.5 0.1>]}}}
     Scale Curve=1.4
    )
    Rotation Vector=<0 0 -1>
   ) {module=UpdateMeshOrientation001}
   &ScaleColor(
    Scale Alpha=&FloatFromCurve(
     FloatCurve=DataInterface() {propertyValues {Curve {keys [
         <at 0 1 RCIM_Cubic>
         {arriveTangent=-4.2795, interpMode=RCIM_Cubic, leaveTangent=-4.2795, tangentMode=RCTM_User, time=0.5844, value=0.5146}
         {arriveTangent=-2.592, interpMode=RCIM_Cubic, leaveTangent=-2.592, tangentMode=RCTM_User, time=1, value=0}
        ]}}}
     Scale Curve=3
    )
    Scale RGB=&Multiply_VectorByFloat(
     Vector=&MakeVectorFromLinearColorRGB(Color=&ColorFromCurve(ColorCurve=DataInterface() {propertyValues {
         AlphaCurve {keys []}
         BlueCurve {keys [<at 0 2 RCIM_Cubic>,<at 0.8853 2 RCIM_Cubic>]}
         GreenCurve {keys [<at 0 0.3 RCIM_Cubic>,<at 0.8853 0.4 RCIM_Cubic>]}
         RedCurve {keys [<at 0 1.7167 RCIM_Cubic>,<at 0.8853 1.4667 RCIM_Cubic>]}
        }}))
     Float=&FloatFromCurve(FloatCurve=DataInterface() {propertyValues {Curve {keys [<at 0 1>,<at 0.5 1>]}}}, Scale Curve=3)
    )
    ScaleA=true
    ScaleRGB=true
   )
   &DynamicMaterialParameters(
    Index 0 Param 1=&FloatFromCurve(FloatCurve=DataInterface() {propertyValues {Curve {keys [<at 0 0 RCIM_Cubic>,<at 1 1>]}}})
    Index 0 Param 2=0.1
    Module.Index0Param0Write=true
    Module.Index0Param1Write=true
    Module.Index0Param2Write=true
    Module.Index0Param3Write=true
   )
  ]
  rendererClasses [&NiagaraMeshRendererProperties]
  renderers [{rendererClass=&NiagaraMeshRendererProperties, rendererIndex=0}]
 }
 SoftAirflow {
  emitterUpdateScript [
   &EmitterState(Life Cycle Mode=::NewEnumerator1, Loop Behavior=::NewEnumerator1, UseLoopDelay=false)
   &SpawnBurst_Instantaneous(Spawn Count=3, Use Spawn Probability=false, UseLoopCountLimit=false)
  ]
  particleSpawnScript [
   &InitializeParticle(
    Color=<rgba 0.8698 0.8698 0.8698 1>
    Lifetime Max=0.7
    Lifetime Min=0.5
    Lifetime Mode=::NewEnumerator1
    Mesh Scale=&Multiply_VectorByFloat(Vector=<1 1 1>, Float=&RandomRangeFloat(Minimum=5, Maximum=5.5))
    Mesh Scale Mode=::NewEnumerator4
    Position Offset=<0 0 200>
    UsePositionOffset=false
    Write Material Random=true
   )
   &InitialMeshOrientation(
    Mesh Orientation Mode=::NewEnumerator7
    Rotation=&RandomRangeVector(Minimum=<0 0 -1>, Maximum=<0 0 1>)
    Use Rotation Vector=true
   )
   &ShapeLocation(
    Cylinder Height=70
    Cylinder Radius=0
    MaskSpawnGroup=false
    Override Seed=false
    Shape Primitive=::NewEnumerator1
    Surface Only=false
    UseOwnerScale=true
   )
  ]
  particleUpdateScript [
   &ParticleState()
   &DynamicMaterialParameters(
    Index 0 Param 1=&FloatFromCurve(FloatCurve=DataInterface() {propertyValues {Curve {keys [{interpMode=RCIM_Cubic, tangentMode=RCTM_User, time=0.5, value=0}, {interpMode=RCIM_Cubic, tangentMode=RCTM_User, time=1, value=1}]}}})
    Index 0 Param 2=&FNTimeOfDayScalar(Night Scale=0, Morning Scale=1, Evening Scale=1)
    Index 0 Param 3=1
    Module.Param0WriteEnabled=true
    Module.Param1WriteEnabled=true
    Module.Param2WriteEnabled=true
    Module.Param3WriteEnabled=true
   )
   &ScaleMeshSize(Scale Factor=&Multiply_VectorByFloat(
     Vector=<1 1 1.5>
     Float=&FloatFromCurve(
      FloatCurve=DataInterface() {propertyValues {Curve {keys [
          {arriveTangent=2.376, interpMode=RCIM_Cubic, leaveTangent=2.376, tangentMode=RCTM_User, time=0, value=0}
          {arriveTangent=0.1286, interpMode=RCIM_Cubic, leaveTangent=0.1286, tangentMode=RCTM_User, time=1, value=1}
         ]}}}
      Scale Curve=1.1
     )
    ))
   &UpdateMeshOrientation(Rotation Vector=<0 0 1>)
   &ScaleColor(
    Scale Alpha=&FloatFromCurve(
     FloatCurve=DataInterface() {propertyValues {Curve {keys [
         {arriveTangent=1.8077, interpMode=RCIM_Cubic, leaveTangent=1.8077, tangentMode=RCTM_User, time=0, value=0}
         {arriveTangent=-1.9397, interpMode=RCIM_Cubic, leaveTangent=-2.9533, tangentMode=RCTM_Break, time=0.41, value=0.5}
         {arriveTangent=-0.2048, interpMode=RCIM_Cubic, leaveTangent=-0.2048, tangentMode=RCTM_User, time=1, value=0}
        ]}}}
     Scale Curve=0.5
    )
    Scale RGB=&MakeVectorFromLinearColorRGB(Color=&FNTimeOfDayColor(Night Color=<rgba 0.3392 0.4002 0.6 1>, Morning Color=<rgba 0.7 0.7 0.7 1>, Day Color=<rgba 1 1 1 1>, Evening Color=<rgba 0.7 0.7 0.7 1>))
    ScaleA=true
    ScaleRGB=true
   )
  ]
  rendererClasses [&NiagaraMeshRendererProperties]
  renderers [{rendererClass=&NiagaraMeshRendererProperties, rendererIndex=0}]
 }
}
summary {
 systemName=NS_Masamune_BigSlashRelease
 emitters [
  {bEnabled=false, emitterName=Reamin, rendererClasses [&NiagaraMeshRendererProperties]}
  {bEnabled=false, emitterName=HighLight, rendererClasses [&NiagaraMeshRendererProperties]}
  {bEnabled=false, emitterName=Glow, rendererClasses [&NiagaraMeshRendererProperties]}
  {emitterName=SoftAirflow, rendererClasses [&NiagaraMeshRendererProperties]}
  {bEnabled=false, emitterName=Glow001, rendererClasses [&NiagaraMeshRendererProperties]}
  {emitterName=FollowParticlesFixed, simTarget=GPUComputeSim, rendererClasses [&NiagaraSpriteRendererProperties]}
  {emitterName=HighLight001, rendererClasses [&NiagaraMeshRendererProperties]}
 ]
 userVariables [User.bIsWanderingBoss {type=NiagaraBool}]
}
